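# Physics asset for the dynamic jacket bones of the large male Tough Henchman character body.
# Seven chains of thin boxes (back, sides, rear and front corners) hang from tiny MSTR capsules linked to the pelvis with free rotation.
# Segments use limited swing and twist; the back forks left and right, while front and side segments pair angled boxes.
# Capsules on pelvis, lower spine and thigh twist bones provide body shapes.
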PhysicsAsset /BRCosmetics/Characters/Player/Male/Large/Bodies/M_LRG_Henchman_Tough/Meshes/M_LRG_Henchman_Tough_jacket_dyn
bodies {
 dyn_jacket_bk [{shapeType=Box, extentX=10, extentY=24, extentZ=2.8068, center=<5 0 0>}]
 dyn_jacket_bk_MSTR [{shapeType=Capsule, length=1, radius=1}]
 dyn_jacket_bk_l_01 [{shapeType=Box, extentX=10, extentY=12, extentZ=1.5, center=<5 3.5 0>}]
 dyn_jacket_bk_r_01 [{shapeType=Box, extentX=10, extentY=12, extentZ=1.5, center=<5 -3.5 0>}]
 dyn_jacket_l_01 [
  {shapeType=Box, extentX=10, extentY=10, extentZ=2, center=<5 -4 0.6>, rotation=<rot 0 0 -5>}
  {shapeType=Box, extentX=10, extentY=10, extentZ=2, center=<5 4 0>, rotation=<rot 0 0 10>}
 ]
 dyn_jacket_l_02 [
  {shapeType=Box, extentX=10, extentY=10, extentZ=2, center=<5 -4 0.6>, rotation=<rot 0 0 -5>}
  {shapeType=Box, extentX=10, extentY=10, extentZ=2, center=<5 4 0>, rotation=<rot 0 0 10>}
 ]
 dyn_jacket_l_MSTR [{shapeType=Capsule, length=1, radius=1}]
 dyn_jacket_l_bk_01 [{shapeType=Box, extentX=10, extentY=16, extentZ=2.5198, center=<5 0 0>}]
 dyn_jacket_l_bk_02 [{shapeType=Box, extentX=10, extentY=16, extentZ=2, center=<5 0 0>}]
 dyn_jacket_l_bk_MSTR [{shapeType=Capsule, length=1, radius=1}]
 dyn_jacket_l_fr_01 [
  {shapeType=Box, extentX=9, extentY=10, extentZ=2, center=<4 0 0.5>}
  {shapeType=Box, extentX=9, extentY=10, extentZ=2, center=<4 -6 0>, rotation=<rot 0 0 -10>}
 ]
 dyn_jacket_l_fr_02 [
  {shapeType=Box, extentX=9, extentY=10, extentZ=2, center=<4 0 0.5>}
  {shapeType=Box, extentX=9, extentY=10, extentZ=2, center=<4 -6 0>, rotation=<rot 0 0 -10>}
 ]
 dyn_jacket_l_fr_03 [
  {shapeType=Box, extentX=10, extentY=10, extentZ=2, center=<4 0 0.5>}
  {shapeType=Box, extentX=10, extentY=10, extentZ=2, center=<4 -6 0>, rotation=<rot 0 0 -10>}
 ]
 dyn_jacket_l_fr_MSTR [{shapeType=Capsule, length=1, radius=1}]
 dyn_jacket_r_01 [
  {shapeType=Box, extentX=10, extentY=10, extentZ=2, center=<5 4 0.3>, rotation=<rot 0 0 5>}
  {shapeType=Box, extentX=10, extentY=10, extentZ=2, center=<5 -4 0>, rotation=<rot 0 0 -10>}
 ]
 dyn_jacket_r_02 [
  {shapeType=Box, extentX=10, extentY=10, extentZ=2, center=<5 4 0.3>, rotation=<rot 0 0 5>}
  {shapeType=Box, extentX=10, extentY=10, extentZ=2, center=<5 -4 0>, rotation=<rot 0 0 -10>}
 ]
 dyn_jacket_r_MSTR [{shapeType=Capsule, length=1, radius=1}]
 dyn_jacket_r_bk_01 [{shapeType=Box, extentX=10, extentY=16, extentZ=2.5198, center=<5 0 0>}]
 dyn_jacket_r_bk_02 [{shapeType=Box, extentX=10, extentY=16, extentZ=2, center=<5 0 0>}]
 dyn_jacket_r_bk_MSTR [{shapeType=Capsule, length=1, radius=1}]
 dyn_jacket_r_fr_01 [
  {shapeType=Box, extentX=9, extentY=10, extentZ=2, center=<4 0 1>}
  {shapeType=Box, extentX=9, extentY=10, extentZ=2, center=<4 9 0>, rotation=<rot 0 0 15>}
 ]
 dyn_jacket_r_fr_02 [
  {shapeType=Box, extentX=9, extentY=10, extentZ=2, center=<4 0 1>}
  {shapeType=Box, extentX=9, extentY=10, extentZ=2, center=<4 9 0>, rotation=<rot 0 0 15>}
 ]
 dyn_jacket_r_fr_03 [
  {shapeType=Box, extentX=10, extentY=10, extentZ=2, center=<4 0 1>}
  {shapeType=Box, extentX=10, extentY=10, extentZ=2, center=<4 9 0>, rotation=<rot 0 0 15>}
 ]
 dyn_jacket_r_fr_MSTR [{shapeType=Capsule, length=1, radius=1}]
 pelvis [{shapeType=Capsule, length=11.5694, radius=13.6524, center=<-6.4768 0.7775 -0.4027>}]
 spine_01 [{shapeType=Capsule, length=8.6779, radius=13.7983, center=<-1.946 1.4313 0>}]
 thigh_twist_01_l [{shapeType=Capsule, length=39.1515, radius=9.4593, center=<-0.0294 1.2525 3.0886>, rotation=<rot 88.1921 4.9998 5.3089>}]
 thigh_twist_01_r [{shapeType=Capsule, length=41.9776, radius=9.3772, center=<-1.7194 -1.7435 -2.3002>, rotation=<rot 90 -0.0049 -0.0049>}]
}
constraints [
 {bone1Name=spine_01, bone2Name=pelvis, swing1LimitDegrees=45, swing2LimitDegrees=45, twistLimitDegrees=45}
 {bone1Name=dyn_jacket_bk, bone2Name=dyn_jacket_bk_MSTR, swing1LimitDegrees=30, swing2LimitDegrees=30, twistLimitDegrees=20}
 {bone1Name=dyn_jacket_bk_l_01, bone2Name=dyn_jacket_bk, swing1LimitDegrees=30, swing2LimitDegrees=30, twistLimitDegrees=20}
 {bone1Name=dyn_jacket_bk_r_01, bone2Name=dyn_jacket_bk, swing1LimitDegrees=30, swing2LimitDegrees=30, twistLimitDegrees=20}
 {bone1Name=dyn_jacket_r_bk_01, bone2Name=dyn_jacket_r_bk_MSTR, swing1LimitDegrees=30, swing2LimitDegrees=30, twistLimitDegrees=20}
 {bone1Name=dyn_jacket_r_bk_02, bone2Name=dyn_jacket_r_bk_01, swing1LimitDegrees=30, swing2LimitDegrees=30, twistLimitDegrees=20}
 {bone1Name=dyn_jacket_l_bk_01, bone2Name=dyn_jacket_l_bk_MSTR, swing1LimitDegrees=30, swing2LimitDegrees=30, twistLimitDegrees=20}
 {bone1Name=dyn_jacket_l_bk_02, bone2Name=dyn_jacket_l_bk_01, swing1LimitDegrees=30, swing2LimitDegrees=30, twistLimitDegrees=20}
 {bone1Name=dyn_jacket_l_fr_01, bone2Name=dyn_jacket_l_fr_MSTR, swing1LimitDegrees=45, swing2LimitDegrees=30, twistLimitDegrees=10}
 {bone1Name=dyn_jacket_l_fr_02, bone2Name=dyn_jacket_l_fr_01, swing1LimitDegrees=35, swing2LimitDegrees=75, twistLimitDegrees=20}
 {bone1Name=dyn_jacket_r_fr_01, bone2Name=dyn_jacket_r_fr_MSTR, swing1LimitDegrees=45, swing2LimitDegrees=30, twistLimitDegrees=20}
 {bone1Name=dyn_jacket_r_fr_02, bone2Name=dyn_jacket_r_fr_01, swing1LimitDegrees=35, swing2LimitDegrees=75, twistLimitDegrees=20}
 {bone1Name=dyn_jacket_l_02, bone2Name=dyn_jacket_l_01, swing1LimitDegrees=30, swing2LimitDegrees=30, twistLimitDegrees=20}
 {bone1Name=dyn_jacket_l_01, bone2Name=dyn_jacket_l_MSTR, swing1LimitDegrees=30, swing2LimitDegrees=30, twistLimitDegrees=20}
 {bone1Name=dyn_jacket_r_01, bone2Name=dyn_jacket_r_MSTR, swing1LimitDegrees=30, swing2LimitDegrees=30, twistLimitDegrees=20}
 {bone1Name=dyn_jacket_r_02, bone2Name=dyn_jacket_r_01, swing1LimitDegrees=30, swing2LimitDegrees=30, twistLimitDegrees=20}
 {bone1Name=dyn_jacket_l_fr_03, bone2Name=dyn_jacket_l_fr_02, swing1LimitDegrees=35, swing2LimitDegrees=75, twistLimitDegrees=20}
 {bone1Name=dyn_jacket_r_fr_03, bone2Name=dyn_jacket_r_fr_02, swing1LimitDegrees=35, swing2LimitDegrees=75, twistLimitDegrees=20}
 {
  bone1Name=dyn_jacket_r_bk_MSTR
  bone2Name=pelvis
  swing1LimitDegrees=45
  swing1Motion=Free
  swing2LimitDegrees=45
  swing2Motion=Free
  twistLimitDegrees=45
  twistMotion=Free
 }
 {
  bone1Name=dyn_jacket_l_bk_MSTR
  bone2Name=pelvis
  swing1LimitDegrees=45
  swing1Motion=Free
  swing2LimitDegrees=45
  swing2Motion=Free
  twistLimitDegrees=45
  twistMotion=Free
 }
 {
  bone1Name=dyn_jacket_bk_MSTR
  bone2Name=pelvis
  swing1LimitDegrees=45
  swing1Motion=Free
  swing2LimitDegrees=45
  swing2Motion=Free
  twistLimitDegrees=45
  twistMotion=Free
 }
 {
  bone1Name=dyn_jacket_r_MSTR
  bone2Name=pelvis
  swing1LimitDegrees=45
  swing1Motion=Free
  swing2LimitDegrees=45
  swing2Motion=Free
  twistLimitDegrees=45
  twistMotion=Free
 }
 {
  bone1Name=dyn_jacket_l_MSTR
  bone2Name=pelvis
  swing1LimitDegrees=45
  swing1Motion=Free
  swing2LimitDegrees=45
  swing2Motion=Free
  twistLimitDegrees=45
  twistMotion=Free
 }
 {
  bone1Name=dyn_jacket_l_fr_MSTR
  bone2Name=pelvis
  swing1LimitDegrees=45
  swing1Motion=Free
  swing2LimitDegrees=45
  swing2Motion=Free
  twistLimitDegrees=45
  twistMotion=Free
 }
 {
  bone1Name=dyn_jacket_r_fr_MSTR
  bone2Name=pelvis
  swing1LimitDegrees=45
  swing1Motion=Free
  swing2LimitDegrees=45
  swing2Motion=Free
  twistLimitDegrees=45
  twistMotion=Free
 }
 {
  bone1Name=thigh_twist_01_r
  bone2Name=pelvis
  swing1LimitDegrees=45
  swing1Motion=Free
  swing2LimitDegrees=45
  swing2Motion=Free
  twistLimitDegrees=45
  twistMotion=Free
 }
 {
  bone1Name=thigh_twist_01_l
  bone2Name=pelvis
  swing1LimitDegrees=45
  swing1Motion=Free
  swing2LimitDegrees=45
  swing2Motion=Free
  twistLimitDegrees=45
  twistMotion=Free
 }
]
