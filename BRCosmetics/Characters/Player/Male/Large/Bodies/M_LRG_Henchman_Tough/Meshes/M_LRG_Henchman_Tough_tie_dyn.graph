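# Physics asset for the necktie on the large male Henchman Tough body, by its names intended for the tie's secondary motion.
# A chain of small capsules runs from dyn_tie_MSTR through five tie bones, constrained link by link and anchored to spine_05.
# Large capsules on the pelvis and five spine bones form a torso volume behind the tie, linked by constraints with free swing and twist.
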PhysicsAsset /BRCosmetics/Characters/Player/Male/Large/Bodies/M_LRG_Henchman_Tough/Meshes/M_LRG_Henchman_Tough_tie_dyn
bodies {
 dyn_tie_01 [{shapeType=Capsule, length=2, radius=3.5, shapeName=dyn_tie_01_capsule, center=<1.5533 0 2.2009>, rotation=<rot 90 0 0>}]
 dyn_tie_02 [{shapeType=Capsule, length=2, radius=3.5, shapeName=dyn_tie_02_capsule, center=<0 0 2.2009>, rotation=<rot 90 0 0>}]
 dyn_tie_03 [{shapeType=Capsule, length=2, radius=3.5, shapeName=dyn_tie_03_capsule, center=<0 0 2.2009>, rotation=<rot 90 0 0>}]
 dyn_tie_04 [{shapeType=Capsule, length=2, radius=3.5, shapeName=dyn_tie_04_capsule, center=<0 0 2.2009>, rotation=<rot 90 0 0>}]
 dyn_tie_05 [{shapeType=Capsule, length=2, radius=3.5, shapeName=dyn_tie_05_capsule, center=<0 0 2.2009>, rotation=<rot 90 0 0>}]
 dyn_tie_MSTR [{
   shapeType=Capsule
   length=1.2162
   radius=0.6621
   shapeName=dyn_tie_MSTR_capsule
   center=<1.6453 0.0001 0.0584>
   rotation=<rot 1.0282 89.8133 -89.9996>
  }]
 pelvis [{shapeType=Capsule, length=4.2506, radius=13.4369, center=<-2.0733 0.2846 0>}]
 spine_01 [{shapeType=Capsule, length=17.0342, radius=14.2158, center=<0 1.7325 0>}]
 spine_02 [{shapeType=Capsule, length=16.6281, radius=14.6219, center=<3.2542 2.5414 0>}]
 spine_03 [{shapeType=Capsule, length=17.048, radius=14.2021, center=<4.1691 2.9491 0>}]
 spine_04 [{shapeType=Capsule, length=16.7976, radius=14.4524, center=<1.4817 2.0834 0>}]
 spine_05 [{shapeType=Capsule, length=20.1895, radius=13.0293, center=<-5.3583 1.2276 0>}]
}
constraints [
 {
  bone1Name=spine_05
  bone2Name=spine_04
  swing1LimitDegrees=45
  swing1Motion=Free
  swing2LimitDegrees=45
  swing2Motion=Free
  twistLimitDegrees=45
  twistMotion=Free
 }
 {
  bone1Name=spine_04
  bone2Name=spine_03
  swing1LimitDegrees=45
  swing1Motion=Free
  swing2LimitDegrees=45
  swing2Motion=Free
  twistLimitDegrees=45
  twistMotion=Free
 }
 {
  bone1Name=spine_03
  bone2Name=spine_02
  swing1LimitDegrees=45
  swing1Motion=Free
  swing2LimitDegrees=45
  swing2Motion=Free
  twistLimitDegrees=45
  twistMotion=Free
 }
 {
  bone1Name=spine_02
  bone2Name=spine_01
  swing1LimitDegrees=45
  swing1Motion=Free
  swing2LimitDegrees=45
  swing2Motion=Free
  twistLimitDegrees=45
  twistMotion=Free
 }
 {
  bone1Name=spine_01
  bone2Name=pelvis
  swing1LimitDegrees=45
  swing1Motion=Free
  swing2LimitDegrees=45
  swing2Motion=Free
  twistLimitDegrees=45
  twistMotion=Free
 }
 {bone1Name=dyn_tie_MSTR, bone2Name=spine_05, swing1LimitDegrees=45, swing2LimitDegrees=45, twistLimitDegrees=45}
 {bone1Name=dyn_tie_01, bone2Name=dyn_tie_MSTR, swing1LimitDegrees=45, swing2LimitDegrees=45, twistLimitDegrees=45}
 {bone1Name=dyn_tie_02, bone2Name=dyn_tie_01, swing1LimitDegrees=45, swing2LimitDegrees=45, twistLimitDegrees=45}
 {bone1Name=dyn_tie_03, bone2Name=dyn_tie_02, swing1LimitDegrees=45, swing2LimitDegrees=45, twistLimitDegrees=45}
 {bone1Name=dyn_tie_04, bone2Name=dyn_tie_03, swing1LimitDegrees=45, swing2LimitDegrees=45, twistLimitDegrees=45}
 {bone1Name=dyn_tie_05, bone2Name=dyn_tie_04, swing1LimitDegrees=45, swing2LimitDegrees=45, twistLimitDegrees=45}
]
