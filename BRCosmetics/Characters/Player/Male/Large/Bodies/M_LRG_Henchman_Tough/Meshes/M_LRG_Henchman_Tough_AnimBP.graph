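# Animation Blueprint for the M_LRG_Henchman_Tough body that copies the attached parent's pose and adds jacket and tie secondary motion.
# Thigh-alpha bone offsets and tie constraints feed two RigidBody simulations, replaced by simple tie constraints when rigid bodies are disabled.
# Movement-state-driven BeltAlpha lifts rear jacket bones after simulation, then backpack and cape attach bones are offset.
# Skydiving, parachute, jump-end and crouch-end events reset dynamics.
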AnimBlueprint /BRCosmetics/Characters/Player/Male/Large/Bodies/M_LRG_Henchman_Tough/Meshes/M_LRG_Henchman_Tough_AnimBP
target_skeleton=&/M_LRG_Henchman_Tough_Skeleton
graphs [
 AnimGraph {nodes [
   n1 = "Constraint - dyn_tie_01 (P)"(ComponentPose=Local To Component(LocalPose=Use cached pose 'Copyposefrommesh'()), Alpha=1, ConstraintWeights_0=1) {
    alphaBoolBlend {blendOption=Linear}
    boneToModify {boneName=dyn_tie_01}
    constraintSetup [{targetBone {boneName=spine_05}}]
    constraintWeights [1]
   }
   n2 = "Constraint - dyn_tie_03 (P)"(ComponentPose=@n1, Alpha=1, ConstraintWeights_0=1) {
    alphaBoolBlend {blendOption=Linear}
    boneToModify {boneName=dyn_tie_03}
    constraintSetup [{targetBone {boneName=spine_04}}]
    constraintWeights [1]
   }
   n3 = "Constraint - dyn_tie_05 (P)"(ComponentPose=@n2, Alpha=1, ConstraintWeights_0=1) {
    alphaBoolBlend {blendOption=Linear}
    boneToModify {boneName=dyn_tie_05}
    constraintSetup [{targetBone {boneName=spine_03}}]
    constraintWeights [1]
   }
   n4 = Component To Local(ComponentPose=@n3)
   n5 = Use cached pose 'RB_Setup'()
   n6 = Get bRigidBodyEnabled()
   n7 = "Transform (Modify) Bone - Bone: attach_backpack"!(
    ComponentPose=Local To Component(LocalPose=Blend Poses by bool(BlendPose_0=@n5, BlendPose_1=@n4, bActiveValue=@n6) {blendType=Linear})
    bAlphaBoolEnabled:bool=True
    Translation="0,-1,0"
    Rotation="0,0,0"
    Scale="1,1,1"
   ) -> (Pose) {
    alphaInputType=Bool
    translationMode=BMM_Additive
    translationSpace=BCS_BoneSpace
    alphaBoolBlend {blendInTime=0.5, blendOutTime=1}
    boneToModify {boneName=attach_backpack}
    translation {y=-1}
   }
   n8 = "Transform (Modify) Bone - Bone: attach_cape"!(ComponentPose=@n7, bAlphaBoolEnabled:bool=True, Translation="0,-2.5,0", Rotation="0,5,0", Scale="1,1,1") -> (Pose) {
    alphaInputType=Bool
    rotationMode=BMM_Additive
    rotationSpace=BCS_BoneSpace
    translationMode=BMM_Additive
    translationSpace=BCS_BoneSpace
    alphaBoolBlend {blendInTime=0.5, blendOutTime=1}
    boneToModify {boneName=attach_cape}
    rotation {yaw=5}
    translation {y=-2.5}
   }
   Output Pose(Result=Component To Local(ComponentPose=@n8)) {layerGroup=None}
   n9 = Get ThighAlphaLeft() -> (ThighAlphaLeft:float)
   n10 = "Transform (Modify) Bone - Bone: dyn_jacket_l_fr_MSTR"(
    ComponentPose=Local To Component(LocalPose=Use cached pose 'Copyposefrommesh'())
    Alpha=@n9
    Translation="-6.850,-1.950,3.601"
    Rotation="0,-30,-25"
    Scale="1,1,1"
   ) {
    rotationMode=BMM_Additive
    rotationSpace=BCS_BoneSpace
    translationMode=BMM_Additive
    translationSpace=BCS_BoneSpace
    alphaBoolBlend {blendOption=Linear}
    boneToModify {boneName=dyn_jacket_l_fr_MSTR}
    rotation {roll=-25, yaw=-30}
    translation=<-6.85 -1.95 3.601>
   }
   n11 = "Transform (Modify) Bone - Bone: dyn_jacket_l_MSTR"(ComponentPose=@n10, Alpha=@n9, Translation="-4,2.501,3.332", Rotation="0.000,15.000,0", Scale="1,1,1") {
    translationMode=BMM_Additive
    translationSpace=BCS_BoneSpace
    alphaBoolBlend {blendOption=Linear}
    boneToModify {boneName=dyn_jacket_l_MSTR}
    translation=<-4 2.501 3.332>
   }
   n12 = Get ThighAlphaRight() -> (ThighAlphaRight:float)
   n13 = "Transform (Modify) Bone - Bone: dyn_jacket_r_fr_MSTR"(ComponentPose=@n11, Alpha=@n12, Translation="4.2,0.000,-5.4", Rotation="0,0,0", Scale="1,1,1") {
    rotationMode=BMM_Additive
    rotationSpace=BCS_BoneSpace
    translationMode=BMM_Additive
    translationSpace=BCS_BoneSpace
    alphaBoolBlend {blendOption=Linear}
    boneToModify {boneName=dyn_jacket_r_fr_MSTR}
    translation {x=4.2, z=-5.4}
   }
   n14 = "Transform (Modify) Bone - Bone: dyn_jacket_r_MSTR"(ComponentPose=@n13, Alpha=@n12, Translation="-3.2,-1.85,2.78", Rotation="0,0,0", Scale="1,1,1") {
    translationMode=BMM_Additive
    translationSpace=BCS_BoneSpace
    alphaBoolBlend {blendOption=Linear}
    boneToModify {boneName=dyn_jacket_r_MSTR}
    translation=<-3.2 -1.85 2.78>
   }
   n15 = "Transform (Modify) Bone - Bone: dyn_jacket_r_fr_MSTR"!(ComponentPose=@n14, Alpha=1, Translation="-1,6,3", Rotation="-15,0,-30") -> (Pose) {
    rotationMode=BMM_Additive
    translationMode=BMM_Additive
    alphaBoolBlend {blendOption=Linear}
    boneToModify {boneName=dyn_jacket_r_fr_MSTR}
    rotation {pitch=-15, roll=-30}
    translation=<-1 6 3>
   }
   n16 = "Constraint - dyn_tie_MSTR (PPP)"!(ComponentPose=@n15, Alpha=1, ConstraintWeights_0=0.827, ConstraintWeights_1:float=0.131, ConstraintWeights_2:float=0.042) -> (Pose) {
    alphaBoolBlend {blendOption=Linear}
    boneToModify {boneName=dyn_tie_MSTR}
    constraintSetup [{targetBone {boneName=spine_05}}, {targetBone {boneName=neck_01}}, {targetBone {boneName=neck_02}}]
    constraintWeights [1,1,1]
   }
   n17 = Break Fort Anim Input Skydiving External Force(FortAnimInput_SkydivingExternalForce=Get SkydivingExternalForceInput()) -> (
    bUseSkydivingVectorForce=~
    SkydivingVectorForce=~
    HeadToPelvisDirection=~
    FloatingMultiplier=~
    FloatingAdditive=~
    DivingMultiplier=~
    DivingAdditive=~
    ParachutingMultiplier=~
    ParachutingAdditive=~
   )
   n18 = Get GravityOverride()
   n19 = RigidBody!(ComponentPose=@n16, Alpha, OverrideWorldGravity:Vector=@n18, ExternalForce=@n17.SkydivingVectorForce) -> (Pose) {
    bClampLinearTranslationLimitToRefPose=true
    bOverrideWorldGravity=true
    overridePhysicsAsset=&/M_LRG_Henchman_Tough_jacket_dyn
    simulationSpace=BaseBoneSpace
    alphaBoolBlend {blendOption=Linear}
    baseBoneRef {boneName=pelvis}
    componentAppliedLinearAccClamp=<3000 3000 3000>
    componentLinearAccScale=<0.3 0.3 0.3>
    componentLinearVelScale=<0.2 0.2 0.2>
    overrideWorldGravity {z=-980}
   }
   n20 = RigidBody(ComponentPose=@n19, Alpha=Get DynamicsAlpha() -> (DynamicsAlpha:float), OverrideWorldGravity=@n18, ExternalForce=@n17.SkydivingVectorForce) {
    bClampLinearTranslationLimitToRefPose=true
    bOverrideWorldGravity=true
    overridePhysicsAsset=&/M_LRG_Henchman_Tough_tie_dyn
    simulationSpace=BaseBoneSpace
    alphaBoolBlend {blendOption=Linear}
    baseBoneRef {boneName=spine_02}
    componentAppliedLinearAccClamp=<3000 3000 3000>
    componentLinearAccScale=<0.4 0.3 0.3>
    componentLinearVelScale=<0.4 0.3 0.3>
    overrideWorldGravity {z=-980}
   }
   n21 = Get BeltAlpha() -> (BeltAlpha:double)
   n22 = "Transform (Modify) Bone - Bone: dyn_jacket_bk_MSTR"(ComponentPose=@n20, Alpha=@n21, Translation="0,0,1", Rotation="0,0,0", Scale="1,1,1") {
    translationMode=BMM_Additive
    translationSpace=BCS_BoneSpace
    alphaBoolBlend {blendOption=Linear}
    alphaScaleBiasClamp {bInterpResult=true}
    boneToModify {boneName=dyn_jacket_bk_MSTR}
    translation {z=1}
   }
   n23 = "Transform (Modify) Bone - Bone: dyn_jacket_r_bk_MSTR"(ComponentPose=@n22, Alpha=@n21, Translation="0,0,0.6", Rotation="0,0,0", Scale="1,1,1") {
    translationMode=BMM_Additive
    translationSpace=BCS_BoneSpace
    alphaBoolBlend {blendOption=Linear}
    alphaScaleBiasClamp {bInterpResult=true}
    boneToModify {boneName=dyn_jacket_r_bk_MSTR}
    translation {z=0.6}
   }
   n24 = "Transform (Modify) Bone - Bone: dyn_jacket_l_bk_MSTR"(ComponentPose=@n23, Alpha=@n21, Translation="0,0,0.6", Rotation="0,0,0", Scale="1,1,1") {
    translationMode=BMM_Additive
    translationSpace=BCS_BoneSpace
    alphaBoolBlend {blendOption=Linear}
    alphaScaleBiasClamp {bInterpResult=true}
    boneToModify {boneName=dyn_jacket_l_bk_MSTR}
    translation {z=0.6}
   }
   Save cached pose 'RB_Setup'(Pose=Component To Local(ComponentPose=@n24)) {cacheName=RB_Setup}
   Save cached pose 'Copyposefrommesh'(Pose=Copy Pose From Mesh() {bUseAttachedParent=true}) {cacheName=Copyposefrommesh}
  ]}
 EventGraph {nodes [
   Event BlueprintUpdateAnimation()
   > n1 = Sequence!(execute) -> (then_0, then_1, then_2:Exec, then_3:Exec)
   > n2 = Branch(Condition=OR Boolean!(
     A=Get bIsTacticalSprinting()
     B=Get bIsSurfaceSwimming()
     C:bool=Get bIsSwimSprinting() -> (bIsSwimSprinting:bool)
     D:bool=Get bIsWaterJump() -> (bIsWaterJump:bool)
     E:bool=Get bIsParachuteOpened()
    ) -> (ReturnValue))
   > Set BeltAlpha(BeltAlpha:double=1) -> (Output_Get:double)
   Set BeltAlpha(execute=@n2.else, BeltAlpha=0)
   n3 = Branch(execute=@n1.then_1, Condition=Get bIsInFrontEnd())
   > Set FrontendAlpha(FrontendAlpha:double=1) -> (Output_Get:double)
   Set FrontendAlpha(execute=@n3.else, FrontendAlpha=0)
   Set FrontendThighRAlpha(
    execute=@n1.then_2
    FrontendThighRAlpha:double=float * float(A=Get ThighAlphaRight(), B=Get FrontendAlpha() -> (FrontendAlpha:double)) {is=title K2Node_PromotableOperator}
   ) -> (Output_Get:double)
   n4 = Get bIsTacticalSprinting()
   n5 = Get bIsJumping()
   Set ExtraBackpackAdjust(execute=@n1.then_3, ExtraBackpackAdjust:bool=OR Boolean!(A=AND Boolean(A=@n4, B=@n5), B=Get bIsSkydiveFloating()) -> (ReturnValue)) -> (Output_Get:bool)
   Event OnParachuteOpened()
   > n6 = Delay(Duration=1.5)
   ResetDynamics(execute=[@n6.then, @n7.then, @n8.then, @n9.then, @n10.then], InTeleportType=TeleportPhysics)
   > ResetDynamics(InTeleportType=ResetPhysics)
   n7 = Event OnBeginSkydiving()
   n8 = Event OnEndJumping()
   n9 = Event OnEndSkydiving()
   n10 = Event OnEndCrouching()
  ]}
]
variables [
 BeltAlpha {type=number}
 ExtraBackpackAdjust {type=boolean}
 FrontendAlpha {type=number}
 FrontendThighRAlpha {type=number}
]
