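# Physics asset supplying cape collision for the large male Henchman Tough player body, judging by its name.
# Capsules approximate the pelvis, spine, clavicles and neck, with several overlapping on spine_02 and spine_05, and a sphere covers the head.
# Arm, leg and foot bodies exist but hold no shapes, so only the torso and head provide collision.
# It has no constraints, so it defines collision volumes without any ragdoll joints.
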PhysicsAsset /BRCosmetics/Characters/Player/Male/Large/Bodies/M_LRG_Henchman_Tough/Meshes/M_LRG_Henchman_Tough_CapeColliders
bodies {
 calf_l []
 calf_r []
 clavicle_l [{shapeType=Capsule, length=7.7636, radius=15.1427, center=<10.293 -0.7741 -0.2799>, rotation=<rot 75.9352 -0.0001 175.5622>}]
 clavicle_r [{shapeType=Capsule, length=7.7636, radius=15.1427, center=<-10.7597 0.7741 -1.5831>, rotation=<rot 75.9351 -0.0001 -4.4378>}]
 foot_l []
 foot_r []
 head [{shapeType=Sphere, radius=17.3125, center=<11.028 -0.174 0>}]
 lowerarm_l []
 lowerarm_r []
 neck_02 [{shapeType=Capsule, length=13.4375, radius=7.3175, center=<-0.2642 -0.3408 0>, rotation=<rot 90 -171.1193 -175.3978>}]
 pelvis [{shapeType=Capsule, length=14.1562, radius=24.5388, center=<-5.7625 1.5098 -1.4851>}]
 spine_01 [{shapeType=Capsule, length=5.7187, radius=17.4138, center=<-2.6215 -0.7156 0>}]
 spine_02 [
  {shapeType=Capsule, length=4.5937, radius=19.445, center=<-2.6215 0.4993 0.0001>}
  {shapeType=Capsule, length=29.875, radius=11.1013, center=<2.5024 0.4992 3.9825>, rotation=<rot 90 0.0049 0.0049>}
  {shapeType=Capsule, length=29.875, radius=11.1013, center=<2.4704 0.4991 -3.9825>, rotation=<rot 90 -0.0396 -0.0396>}
 ]
 spine_03 [{shapeType=Capsule, length=5.75, radius=21.2575, center=<-1.953 0.1364 0.0001>}]
 spine_04 [{shapeType=Capsule, length=12.1562, radius=19.6325, center=<-0.8013 -1.2748 0.0002>}]
 spine_05 [
  {shapeType=Capsule, length=25.0937, radius=21.695, center=<-1.7514 2.4879 0>}
  {shapeType=Capsule, length=12.7812, radius=17.0625, center=<0.7875 3.1143 0>}
 ]
 thigh_l []
 thigh_r []
 upperarm_l []
 upperarm_r []
}
constraints []
