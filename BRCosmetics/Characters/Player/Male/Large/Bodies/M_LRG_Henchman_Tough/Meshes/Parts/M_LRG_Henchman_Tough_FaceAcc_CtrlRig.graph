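# Control Rig for the Tough Henchman's face accessory that procedurally adjusts glasses bones in response to cheek and brow motion.
# Local cheek and brow offsets from initial pose drive it: half the higher cheek Z lifts glasses_c.
# glasses_l then aims at glasses_r, using glasses_c as secondary target to tilt, and shifts by half any positive center-brow Y.
# Two float Min nodes are left unconnected to the result.
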ControlRigBlueprint /BRCosmetics/Characters/Player/Male/Large/Bodies/M_LRG_Henchman_Tough/Meshes/Parts/M_LRG_Henchman_Tough_FaceAcc_CtrlRig
graphs [
 RigVMFunctionLibrary {links [], nodes []}
 RigVMModel {nodes [
   n1 = &RigUnit_BeginExecution()
   n2 = &RigVMFunction_MathTransformMul(
    A=&RigUnit_GetBoneTransform(Bone=glasses_cheek_l, Space=LocalSpace).Transform
    B=/Script/RigVM.RigVMFunction_MathTransformInverse(Value=/Script/ControlRig.RigUnit_GetInitialBoneTransform(Bone=glasses_cheek_l, Space=LocalSpace) {is=struct RigVMUnitNode}.Transform) {is=struct RigVMUnitNode, template="Inverse::Execute(in Value,out Result)"}.Result
   ) {template="Multiply::Execute(in A,in B,out Result)"}
   n3 = &RigVMFunction_MathTransformMul(
    A=&RigUnit_GetBoneTransform(Bone=glasses_cheek_r, Space=LocalSpace).Transform
    B=&RigVMFunction_MathTransformInverse(Value=&RigUnit_GetInitialBoneTransform(Bone=glasses_cheek_r, Space=LocalSpace).Transform) {template="Inverse::Execute(in Value,out Result)"}.Result
   ) {template="Multiply::Execute(in A,in B,out Result)"}
   /Script/ControlRig.RigUnit_AddBoneTransform(
    Bone=glasses_c
    Transform=(Rotation=(X=0,Y=0,Z=0,W=1),Translation=(X=0,Y=0,Z=0),Scale3D=(X=1,Y=1,Z=1))
    Weight=1
    bPostMultiply=true
    bPropagateToChildren=false
    ExecutePin=@n1.ExecutePin
    Transform.Translation.Z=&RigVMFunction_MathFloatMul(
     B=0.5
     A=/Script/RigVM.RigVMFunction_MathFloatMax(A=@n2.Result.Translation.Z, B=@n3.Result.Translation.Z) {is=struct RigVMUnitNode, template="Maximum::Execute(in A,in B,out Result)"}.Result
    ) {template="Multiply::Execute(in A,in B,out Result)"}.Result
   ) {is=struct RigVMUnitNode}
   > /Script/ControlRig.RigUnit_AimBone(
    Bone=glasses_l
    DebugSettings=(bEnabled=False,Scale=10,WorldOffset=(Rotation=(X=0,Y=0,Z=0,W=1),Translation=(X=0,Y=0,Z=0),Scale3D=(X=1,Y=1,Z=1)))
    Primary=(Weight=1,Axis=(X=-1,Y=0,Z=0),Target=(X=0,Y=0,Z=0),Kind=Location,Space="None")
    Secondary=(Weight=1,Axis=(X=0,Y=1,Z=0),Target=(X=0,Y=0,Z=0),Kind=Location,Space="None")
    Weight=1
    bPropagateToChildren=True
    Primary.Target=&RigUnit_GetBoneTransform(Bone=glasses_r).Transform.Translation
    Secondary.Target=&RigUnit_GetBoneTransform(Bone=glasses_c).Transform.Translation
   ) {is=struct RigVMUnitNode}
   n4 = &RigVMFunction_MathTransformMul(
    A=&RigUnit_GetBoneTransform(Bone=glasses_brow_c, Space=LocalSpace).Transform
    B=&RigVMFunction_MathTransformInverse(Value=&RigUnit_GetInitialBoneTransform(Bone=glasses_brow_c, Space=LocalSpace).Transform) {template="Inverse::Execute(in Value,out Result)"}.Result
   ) {template="Multiply::Execute(in A,in B,out Result)"}
   > &RigUnit_AddBoneTransform(
    Bone=glasses_l
    Transform=(Rotation=(X=0,Y=0,Z=0,W=1),Translation=(X=0,Y=0,Z=0),Scale3D=(X=1,Y=1,Z=1))
    Weight=1
    bPostMultiply=true
    bPropagateToChildren=false
    Transform.Translation.Y=&RigVMFunction_MathFloatMul(
     B=0.5
     A=&RigVMFunction_MathFloatMax(B=0, A=@n4.Result.Translation.Y) {template="Maximum::Execute(in A,in B,out Result)"}.Result
    ) {template="Multiply::Execute(in A,in B,out Result)"}.Result
   )
   /Script/RigVM.RigVMFunction_MathFloatMin(A=0, B=0) {is=struct RigVMUnitNode, template="Minimum::Execute(in A,in B,out Result)"}
   &RigVMFunction_MathFloatMin(B=0, A=@n4.Result.Translation.Y) {template="Minimum::Execute(in A,in B,out Result)"}
  ]}
]
hierarchy [
 root {loc=<0 0 0>, rot=<0 0 0 1>}
 pelvis {parent=root, loc=<0 -0.7321 92.4218>, rot=<0.0127 0.707 -0.0127 -0.707>}
 spine_01 {parent=pelvis, loc=<5.6002 -0.5128 0>, rot=<0 0 0.034 -0.9994>}
 spine_02 {parent=spine_01, loc=<8.0753 0.0139 0>, rot=<0 0 0.0238 -0.9997>}
 spine_03 {parent=spine_02, loc=<7.0393 0 0>, rot=<0 0 0.0149 -0.9999>}
 spine_04 {parent=spine_03, loc=<8.1024 0 0>, rot=<0 0 -0.0132 -0.9999>}
 spine_05 {parent=spine_04, loc=<10.0146 0 0>, rot=<0 0 -0.0015 -1>}
 neck_01 {parent=spine_05, loc=<10.815 1.9196 0>, rot=<0 0 -0.1864 -0.9825>}
 neck_02 {parent=neck_01, loc=<4.7347 0.3045 0>, rot=<0 0 0.0363 -0.9993>}
 head {parent=neck_02, loc=<5.7632 0.5552 0>, rot=<0 0 0.115 -0.9934>}
 faceAttach {parent=head, loc=<0.8774 0.4835 0>, rot=<0.0082 -0.7071 -0.0082 -0.7071>, scale=<0.89 0.89 0.89>}
 L_cheek_inner {parent=faceAttach, loc=<4.7608 6.8808 0.1737>, rot=<-0.0706 -0.2697 0.3184 -0.906>}
 C_lip_upper_mid {parent=faceAttach, loc=<0 9.566 -2.5212>, rot=<0.2225 0.0013 -0.0005 -0.9749>}
 L_lip_upper_outer {parent=faceAttach, loc=<1.8851 8.9589 -2.5848>, rot=<0.0359 -0.0728 0.2627 -0.9615>}
 L_lip_corner {parent=faceAttach, loc=<3.0217 8.0312 -2.7974>, rot=<0.003 0.008 -0.0146 -0.9999>}
 C_nose_bridge {parent=faceAttach, loc=<0 10.8736 1.5016>, rot=<-0.2218 0 0 -0.9751>}
 R_cheek_inner {parent=faceAttach, loc=<-4.7608 6.8808 0.1737>, rot=<-0.0706 0.2697 -0.3184 -0.906>}
 R_lip_corner {parent=faceAttach, loc=<-3.0217 8.0312 -2.7974>, rot=<0.003 -0.008 0.0146 -0.9999>}
 C_jaw {parent=faceAttach, loc=<-0.0151 1.2451 -0.5092>, rot=<0 0 0 1>}
 R_lip_upper_outer {parent=faceAttach, loc=<-1.8851 8.9589 -2.5848>, rot=<0.0359 0.0728 -0.2627 -0.9615>}
 C_brow_mid {parent=faceAttach, loc=<0 9.4013 5.6348>, rot=<0 0 0 -1>}
 L_brow_mid {parent=faceAttach, loc=<3.3704 8.3995 6.093>, rot=<-0.0001 -0.0043 0.2553 -0.9669>}
 R_brow_mid {parent=faceAttach, loc=<-3.3704 8.3995 6.093>, rot=<-0.0001 0.0043 -0.2553 -0.9669>}
 glasses {parent=head, loc=<5.1393 0.621 0>, rot=<0.0029 -0.7071 -0.0029 -0.7071>}
 glasses_c {parent=glasses, loc=<0 8.3587 0>, rot=<0 0 0 -1>}
 glasses_l {parent=glasses, loc=<5.8593 0 0>, rot=<0 0 0 -1>}
 glasses_r {parent=glasses, loc=<-5.8593 0 0>, rot=<0 0 0 1>}
 glasses_front {parent=glasses, loc=<0 7.3357 -0.5005>, rot=<0 0 0 1>}
 glasses_cheek_l {parent=glasses, loc=<2.7287 7.0403 -3.5369>, rot=<0 0 0 1>}
 glasses_cheek_r {parent=glasses, loc=<-2.7287 7.0403 -3.5369>, rot=<0 0 0 1>}
 glasses_brow_c {parent=glasses, loc=<0 8.3382 0.6277>, rot=<0 0 0 1>}
]
