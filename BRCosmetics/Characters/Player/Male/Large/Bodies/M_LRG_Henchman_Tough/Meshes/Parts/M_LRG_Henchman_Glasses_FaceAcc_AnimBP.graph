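# Post-style face accessory AnimBlueprint for the Large Male Tough Henchman's glasses, fitting them to the face as it moves.
# It copies the attached parent mesh's pose, then weighted multi-target constraints drive the left cheek, right cheek and center brow glasses bones from head, lip, jaw, cheek, nose and brow bones.
# A face accessory Control Rig runs last.
# Constraints and rig are enabled only at LOD 0, checked every update.
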AnimBlueprint /BRCosmetics/Characters/Player/Male/Large/Bodies/M_LRG_Henchman_Tough/Meshes/Parts/M_LRG_Henchman_Glasses_FaceAcc_AnimBP
target_skeleton=&/M_LRG_Henchman_Glasses_FaceAcc_Skeleton
graphs [
 AnimGraph {nodes [
   n1 = Get isLOD0() -> (isLOD0:bool)
   n2 = "Constraint - glasses_cheek_l (PPPPPPP)"!(
    ComponentPose=Local To Component(LocalPose=Copy Pose From Mesh() {bUseAttachedParent=true})
    bAlphaBoolEnabled:bool=@n1
    ConstraintWeights_0=0.1536
    ConstraintWeights_1:float=0.0243
    ConstraintWeights_2:float=0.006
    ConstraintWeights_3:float=0.0471
    ConstraintWeights_4:float=0.0069
    ConstraintWeights_5:float=0.6549
    ConstraintWeights_6:float=0.1072
   ) -> (Pose) {
    alphaInputType=Bool
    alphaBoolBlend {blendOption=Linear}
    boneToModify {boneName=glasses_cheek_l}
    constraintSetup [
     {targetBone {boneName=head}}
     {targetBone {boneName=L_lip_corner}}
     {targetBone {boneName=C_lip_upper_mid}}
     {targetBone {boneName=L_lip_upper_outer}}
     {targetBone {boneName=C_jaw}}
     {targetBone {boneName=L_cheek_inner}}
     {targetBone {boneName=C_nose_bridge}}
    ]
    constraintWeights [1,1,1,1,1,1,1]
   }
   n3 = "Constraint - glasses_cheek_r (PPPPPPP)"!(
    ComponentPose=@n2
    bAlphaBoolEnabled:bool=@n1
    ConstraintWeights_0=0.1536
    ConstraintWeights_1:float=0.0471
    ConstraintWeights_2:float=0.006
    ConstraintWeights_3:float=0.0243
    ConstraintWeights_4:float=0.0069
    ConstraintWeights_5:float=0.1072
    ConstraintWeights_6:float=0.6549
   ) -> (Pose) {
    alphaInputType=Bool
    alphaBoolBlend {blendOption=Linear}
    boneToModify {boneName=glasses_cheek_r}
    constraintSetup [
     {targetBone {boneName=head}}
     {targetBone {boneName=R_lip_upper_outer}}
     {targetBone {boneName=C_lip_upper_mid}}
     {targetBone {boneName=R_lip_corner}}
     {targetBone {boneName=C_jaw}}
     {targetBone {boneName=C_nose_bridge}}
     {targetBone {boneName=R_cheek_inner}}
    ]
    constraintWeights [1,1,1,1,1,1,1]
   }
   n4 = "Constraint - glasses_brow_c (PPP)"!(
    ComponentPose=@n3
    bAlphaBoolEnabled:bool=@n1
    ConstraintWeights_0=0.9438
    ConstraintWeights_1:float=0.0391
    ConstraintWeights_2:float=0.0171
   ) -> (Pose) {
    alphaInputType=Bool
    alphaBoolBlend {blendOption=Linear}
    boneToModify {boneName=glasses_brow_c}
    constraintSetup [{targetBone {boneName=C_brow_mid}}, {targetBone {boneName=head}}, {targetBone {boneName=L_brow_mid}}]
    constraintWeights [1,1,1]
   }
   n5 = Control Rig!(Source=Component To Local(ComponentPose=@n4), bAlphaBoolEnabled:bool=@n1) -> (Pose) {
    alphaInputType=Bool
    alphaBoolBlend {blendOption=Linear}
    controlRigAssetReference {blueprintRigClass=&/M_LRG_Henchman_Tough_FaceAcc_CtrlRig.M_LRG_Henchman_Tough_FaceAcc_CtrlRig_C}
   }
   Output Pose(Result=@n5) {layerGroup=None}
  ]}
 EventGraph {nodes [
   Event BlueprintUpdateAnimation()
   n1 = Get CurrentLOD()
   > Set isLOD0(isLOD0:bool=NOT Boolean(A="To Boolean (Integer)"!(InInt:int=@n1) -> (ReturnValue:bool))) -> (Output_Get:bool)
   TryGetPawnOwner()
  ]}
]
variables [isLOD0 {type=boolean}]
